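; Projectile Blueprint for an Athena AI ranged NPC attack that breaks on impact or when it stops and triggers an explosion.
; On hit or server-side stop, an Exploded flag guards one-time logic that spawns a break particle, disables capsule collision, deactivates the projectile particles and calls DoExplosion.
; The hit path also stops movement and orients the particle to the surface normal.
; On explosion it makes AI noise, plays an attached impact sound and fades out the projectile audio.
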(event EventOnStoppedOnServer
  ;[1 1 1] EVENT -> OnStoppedOnServer
  (if (not (Variables|Default|GetExploded))
    (Variables|Default|SetExploded true)
    (Effects|SpawnEmitteratLocation (Variables|Default|GetPS_ProjectileBreak) (Transformation|GetActorLocation) (Transformation|GetActorRotation))
    (Collision|SetCollisionEnabled (Variables|Collision|GetCapsuleComponent) "NoCollision")
    (Components|Activation|Deactivate (Variables|ProjectileDeimosRanged|GetProjectileParticleSystem))
    (Delivery|DoExplosion)))

(event Collision|EventHit
  (params (MyComp :type Primitive%20Component%20Object%20Reference) (Other :type Actor%20Object%20Reference) (OtherComp :type Primitive%20Component%20Object%20Reference) (bSelfMoved :type Boolean) (HitLocation :type Vector :default "0, 0, 0") (HitNormal :type Vector :default "0, 0, 0") (NormalImpulse :type Vector :default "0, 0, 0") (Hit :type Hit%20Result%20Structure%20%28by%20ref%29))
  ;[1 1 1] EVENT -> Hit
  (if (not (Variables|Default|GetExploded))
    (Variables|Default|SetExploded true)
    (Effects|SpawnEmitteratLocation (Variables|Default|GetPS_ProjectileBreak) HitLocation (Math|Rotator|MakeRotfromX HitNormal))
    (Game|Components|ProjectileMovement|StopSimulating (Variables|Movement|GetProjectileMovementComponent) Hit)
    (Collision|SetCollisionEnabled (Variables|Collision|GetCapsuleComponent) "NoCollision")
    (Components|Activation|Deactivate (Variables|ProjectileDeimosRanged|GetProjectileParticleSystem))
    (Delivery|DoExplosion)))

(event EventOnExploded
  (params (HitActors :type Array%20of%20Actor%20Object%20References) (HitResults :type Array%20of%20Hit%20Result%20Structures))
  ;[1 1 1] EVENT -> OnExploded
  (AI|MakeNoiseAtAIProjectile
    (:Then
      (Audio|SpawnSoundAttached (Variables|Default|GetProjectielDirtImpact_Cue) (Variables|ProjectileDeimosRanged|GetRanged_Projectile_Sound))
      (Audio|Components|Audio|FadeOut (Variables|ProjectileDeimosRanged|GetRanged_Projectile_Sound) 0.2 0.0 "Linear" :!class /Script/Engine.AudioComponent))))

(fn ConstructionScript)
